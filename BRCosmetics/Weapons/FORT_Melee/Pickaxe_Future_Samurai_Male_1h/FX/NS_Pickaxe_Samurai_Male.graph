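# Niagara system for the Future Samurai Male pickaxe cosmetic, adding periodic glints to the blade through a supplied skeletal mesh.
# Sharp_blade_attack bursts one short-lived, user-colored mesh particle per randomized loop, placed by skeletal mesh sampling, with a curve-animated material parameter and boosted brightness.
# Flash_attack defines an offset, curve-scaled sprite flash, but its only spawn module is disabled.
# The two Blade_flash user floats are not referenced by the modules shown.
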NiagaraSystem /BRCosmetics/Weapons/FORT_Melee/Pickaxe_Future_Samurai_Male_1h/FX/NS_Pickaxe_Samurai_Male
dependencies {usedDataInterfaces [&NiagaraDataInterfaceCurve], usedDynamicInputs [], usedModules [&SystemState]}
emitters {
 Flash_attack {
  emitterUpdateScript [
   &EmitterState(Inactive Response=::NewEnumerator1, Life Cycle Mode=::NewEnumerator1, Loop Duration=1, UseLoopDelay=false)
   &SpawnBurst_Instantaneous(Spawn Count=1, Use Spawn Probability=false) {enabled=false}
  ]
  particleSpawnScript [/Niagara/Modules/Spawn/Initialization/InitializeParticle(
    Color=<rgba 0.6 0.6 0.6 1>
    Lifetime=0.25
    Module.Write Color=true
    Module.Write Lifetime=true
    Module.Write Mass=true
    Module.Write Material Random=true
    Module.Write Position=false
    Module.Write Scale=false
    Module.Write SpriteRotation=true
    Module.Write SpriteSize=true
    Position=&SimulationPosition()
    Sprite Rotation=&UniformRangedFloat(Minimum=0, Maximum=1)
    Sprite Size=&Vector2DFromFloat(Value=30)
   )]
  particleUpdateScript [
   &ParticleState()
   &SolveForcesAndVelocity(Clamp Velocity=false, Limit Acceleration=false)
   &Color(
    Color=<rgba 0.4306 0.6707 0.7188 1>
    Scale Color=&VectorFromFloat(Value=&FloatFromCurve(FloatCurve=DataInterface() {propertyValues {Curve {keys [<at 0 5>,<at 0.3 1>]}}}, Scale Curve=2))
   )
   &ScaleSpriteSize(Scale Factor=&Vector2DFromFloat(Value=&FloatFromCurve(FloatCurve=DataInterface() {propertyValues {Curve {keys [
          <at -0.0035 0.0142 RCIM_Cubic>
          {arriveTangent=0.6433, interpMode=RCIM_Cubic, leaveTangent=0.6433, time=0.2414, value=0.9767}
          {arriveTangent=-1.2875, interpMode=RCIM_Cubic, leaveTangent=-1.2875, time=0.481, value=0.3258}
          <at 1 0>
         ]}}})))
   &SpriteRotationRate(Delta Time:NiagaraFloat=$Particles.Lifetime, Rotation Rate=30)
   &CameraOffset(Camera Offset Amount=15)
   &ScaleColor(Scale RGB=<2.5 2.5 2.5>, ScaleA=true, ScaleRGB=true)
   &SkeletalMeshLocation(
    Bone Sampling Mode=::NewEnumerator0
    Inherit Velocity=false
    MaskSpawnGroup=false
    Module.Override Randomness=false
    Sampled Position Offset=<0 0 56>
    Sampled Position Offset Coordinate Space:ENiagaraMeshTransforms=::NewEnumerator2
    Skeletal Mesh=$User.SkeletalMesh
   )
  ]
  rendererClasses [&NiagaraSpriteRendererProperties]
  renderers [{rendererClass=&NiagaraSpriteRendererProperties}]
 }
 Sharp_blade_attack {
  emitterUpdateScript [
   &EmitterState(
    Delay First Loop Only=true
    Inactive Response=::NewEnumerator1
    Life Cycle Mode=::NewEnumerator1
    Loop Delay=&UniformRangedFloat(Minimum=0, Maximum=0.3)
    Loop Duration=&RandomRangeFloat(Minimum=2, Maximum=3)
    Recalculate Duration Each Loop=true
    UseLoopDelay=true
   )
   &SpawnBurst_Instantaneous(Spawn Count=1, Use Spawn Probability=false)
  ]
  particleSpawnScript [&InitializeParticle(
    Color=$User.Color
    Lifetime=&RandomRangeFloat(Minimum=0.2, Maximum=0.4)
    Mesh Scale=&Multiply_Vector(A=<1 1 1>, B=<1 1 1>)
    Module.Write Color=true
    Module.Write Lifetime=true
    Module.Write Mass=true
    Module.Write Material Random=true
    Module.Write Position=false
    Module.Write Scale=true
    Module.Write SpriteRotation=false
    Module.Write SpriteSize=false
    Position=&SimulationPosition()
    Sprite Size=<5 5>
   )]
  particleUpdateScript [
   &ParticleState()
   &SolveForcesAndVelocity(Clamp Velocity=false, Limit Acceleration=false)
   &DynamicMaterialParameters(
    Index 0 Param 1=&FloatFromCurve(FloatCurve=DataInterface() {propertyValues {Curve {keys [<at 0 0>,<at 1 1.7>]}}})
    Module.Param0WriteEnabled=true
    Module.Param1WriteEnabled=true
    Module.Param2WriteEnabled=true
    Module.Param3WriteEnabled=true
   )
   &GenerateDeathEvent()
   &CameraOffset(Camera Offset Amount=1)
   &ScaleColor(Scale RGB=<10 10 10>, ScaleA=true, ScaleRGB=true)
   &SkeletalMeshLocation(
    Bone Sampling Mode=::NewEnumerator0
    Inherit Velocity=false
    MaskSpawnGroup=false
    Module.Override Randomness=false
    Skeletal Mesh=$User.SkeletalMesh
   )
  ]
  rendererClasses [&NiagaraMeshRendererProperties]
  renderers [{rendererClass=&NiagaraMeshRendererProperties, rendererIndex=0}]
 }
}
summary {
 systemName=NS_Pickaxe_Samurai_Male
 emitters [
  {emitterName=Sharp_blade_attack, rendererClasses [&NiagaraMeshRendererProperties]}
  {emitterName=Flash_attack, rendererClasses [&NiagaraSpriteRendererProperties]}
 ]
 userVariables [
  User.Blade_flash_attack {type=NiagaraFloat, defaultValue=0, description=""}
  User.Blade_flash_loop {type=NiagaraFloat, defaultValue=1, description=""}
  User.Color {type=LinearColor, description="", defaultValue {struct=LinearColor, value=<rgba 1 1 1 1>}}
  User.SkeletalMesh {
   type=NiagaraDataInterfaceSkeletalMesh
   defaultValue {
    struct=NiagaraExt_VariableValue_DataInterface
    value {dataInterface=&self:NiagaraDataInterfaceSkeletalMesh_0}
   }
  }
 ]
}
